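# Milkshake sound cue in the earthquake sound class, layering a rumble start with close and distant spline loops and fissure steam bursts.
# The spline loops and steam layer fade in through volume envelopes, each with close or distant attenuation.
# Random close (pitch-varied) and distant start one-shots mix separately behind a quality switch that fills only its first two slots.
# A root modulator wraps the final mix.
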SoundCue /Milkshake/Sounds/Canary_Milkshake_Loop_Cue
root=Modulator_5
nodes [
 {id=Attenuation_0, children [Modulator_0], properties {attenuationSettings=&/Milkshake_Loop_Close_Att}}
 {id=Attenuation_1, children [Random_1], properties {attenuationSettings=&/Milkshake_Loop_Distant_Att}}
 {id=Attenuation_2, children [WavePlayer_3], properties {attenuationSettings=&Milkshake_Loop_Distant_Att}}
 {id=Attenuation_3, children [WavePlayer_2], properties {attenuationSettings=&Milkshake_Loop_Close_Att}}
 {id=Attenuation_4, children [WavePlayer_1], properties {attenuationSettings=&Milkshake_Loop_Close_Att}}
 {
  id=Enveloper_0
  children [Attenuation_3]
  properties {pitchCurve {editorCurveData {keys [{time=0}]}}, volumeCurve {editorCurveData {keys [<at 0 0.01>,<at 8 1>]}}}
 }
 {
  id=Enveloper_1
  children [Attenuation_2]
  properties {pitchCurve {editorCurveData {keys [{time=0}]}}, volumeCurve {editorCurveData {keys [<at 0 0.01>,<at 8 1>]}}}
 }
 {
  id=Enveloper_2
  children [Attenuation_4]
  properties {pitchCurve {editorCurveData {keys [{time=0}]}}, volumeCurve {editorCurveData {keys [<at 0 0.01>,<at 8 1>]}}}
 }
 {id=Mixer_0, children [Enveloper_0, Enveloper_1, QualityLevel_2, Enveloper_2], properties {inputVolume [2,0.8,3,2]}}
 {id=Mixer_1, children [Attenuation_0, Attenuation_1], properties {inputVolume [0.2,0.3]}}
 {id=Modulator_0, children [Random_0], properties {pitchMax=1.2, pitchMin=0.9, volumeMax=1, volumeMin=1}}
 {id=Modulator_5, children [Mixer_0], properties {volumeMax=1, volumeMin=1}}
 {id=QualityLevel_2, children [Mixer_1,Mixer_1,null,null]}
 {id=Random_0, children [WavePlayer_4,WavePlayer_6], properties {bShouldExcludeFromBranchCulling=true, weights [1,1]}}
 {id=Random_1, children [WavePlayer_7,WavePlayer_8,WavePlayer_9], properties {bShouldExcludeFromBranchCulling=true, weights [1,1,1]}}
 {id=WavePlayer_1, properties {soundWaveAssetPtr=&/Milkshake_Fissures_SteamBursts_Loop}}
 {id=WavePlayer_2, properties {bLooping=true, soundWaveAssetPtr=&/milkshake_spline_close_loop}}
 {id=WavePlayer_3, properties {bLooping=true, soundWaveAssetPtr=&/milkshake_spline_distant_loop}}
 {id=WavePlayer_4, properties {soundWaveAssetPtr=&/milkshake_start_close_01}}
 {id=WavePlayer_6, properties {soundWaveAssetPtr=&/milkshake_start_close_03}}
 {id=WavePlayer_7, properties {soundWaveAssetPtr=&/milkshake_start_distant_01}}
 {id=WavePlayer_8, properties {soundWaveAssetPtr=&/milkshake_start_distant_02}}
 {id=WavePlayer_9, properties {soundWaveAssetPtr=&/milkshake_start_distant_03}}
]
properties {maxDistance=16000, soundClassObject=/Game/Sounds/FortSoundClassesAndModes/Fort_Earthquakes, volumeMultiplier=1.5}
