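# Sound cue for a concrete crack, built from fissure burst waves and routed to the earthquakes sound class.
# It mixes a random "a" burst with a random "b" burst delayed by a further random offset, each with pitch and volume modulation, under a short overall random delay.
# The quality node plays this chain on the first two levels and nothing on the other two.
# Attenuation is shared rattle settings, and concurrency is capped, rejecting new plays.
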SoundCue /Milkshake/Sounds/Milkshake_Concrete_Crack_Cue
root=QualityLevel_1
nodes [
 {id=Delay_0, children [Mixer_0], properties {delayMax=0.1}}
 {id=Delay_1, children [Modulator_2], properties {delayMax=0.5}}
 {id=Mixer_0, children [Modulator_1,Delay_1], properties {inputVolume [1,1]}}
 {id=Modulator_0, children [Delay_0], properties {pitchMax=1, pitchMin=1, volumeMax=1, volumeMin=0.8}}
 {id=Modulator_1, children [Random_1], properties {pitchMax=1.2}}
 {id=Modulator_2, children [Random_2], properties {pitchMax=1.2, volumeMin=0.3}}
 {id=QualityLevel_1, children [Modulator_0,Modulator_0,null,null]}
 {
  id=Random_1
  children [WavePlayer_7, WavePlayer_8, WavePlayer_9, WavePlayer_10, WavePlayer_11, WavePlayer_12, WavePlayer_13, WavePlayer_14]
  properties {bShouldExcludeFromBranchCulling=true, weights [1,1,1,1,1,1,1,1]}
 }
 {
  id=Random_2
  children [WavePlayer_15, WavePlayer_16, WavePlayer_17, WavePlayer_18, WavePlayer_19, WavePlayer_20, WavePlayer_21]
  properties {bShouldExcludeFromBranchCulling=true, weights [1,1,1,1,1,1,1]}
 }
 {id=WavePlayer_10, properties {soundWaveAssetPtr=&/milkshake_fissure_burst_a_04}}
 {id=WavePlayer_11, properties {soundWaveAssetPtr=&/milkshake_fissure_burst_a_05}}
 {id=WavePlayer_12, properties {soundWaveAssetPtr=&/milkshake_fissure_burst_a_06}}
 {id=WavePlayer_13, properties {soundWaveAssetPtr=&/milkshake_fissure_burst_a_07}}
 {id=WavePlayer_14, properties {soundWaveAssetPtr=&/milkshake_fissure_burst_a_08}}
 {id=WavePlayer_15, properties {soundWaveAssetPtr=&/milkshake_fissure_burst_b_01}}
 {id=WavePlayer_16, properties {soundWaveAssetPtr=&/milkshake_fissure_burst_b_02}}
 {id=WavePlayer_17, properties {soundWaveAssetPtr=&/milkshake_fissure_burst_b_03}}
 {id=WavePlayer_18, properties {soundWaveAssetPtr=&/milkshake_fissure_burst_b_04}}
 {id=WavePlayer_19, properties {soundWaveAssetPtr=&/milkshake_fissure_burst_b_05}}
 {id=WavePlayer_20, properties {soundWaveAssetPtr=&/milkshake_fissure_burst_b_06}}
 {id=WavePlayer_21, properties {soundWaveAssetPtr=&/milkshake_fissure_burst_b_07}}
 {id=WavePlayer_7, properties {soundWaveAssetPtr=&/milkshake_fissure_burst_a_01}}
 {id=WavePlayer_8, properties {soundWaveAssetPtr=&/milkshake_fissure_burst_a_02}}
 {id=WavePlayer_9, properties {soundWaveAssetPtr=&/milkshake_fissure_burst_a_03}}
]
properties {
 attenuationSettings=&/Milkshake_Rattle_Stereo_Att
 bOverrideConcurrency=true
 maxDistance=2000
 soundClassObject=/Game/Sounds/FortSoundClassesAndModes/Fort_Earthquakes
 volumeMultiplier=1.3
 concurrencyOverrides {maxCount=3, resolutionRule=PreventNew}
}
