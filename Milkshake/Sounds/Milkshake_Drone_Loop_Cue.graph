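# Sound cue for the looping Milkshake ambient drone, routed through the Fort_Earthquakes sound class.
# It mixes two looping waves, a louder drone with close attenuation and a quieter 2D quad version with its own attenuation.
# An Enveloper over the mix fades volume in over three seconds.
# A concurrency override allows one instance, stopping the farthest first, with a three-second release when a voice is stolen.
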SoundCue /Milkshake/Sounds/Milkshake_Drone_Loop_Cue
root=Enveloper_2
nodes [
 {id=Attenuation_11, children [WavePlayer_36], properties {attenuationSettings=&/Milkshake_Loop_Drone_Close_Att}}
 {id=Attenuation_12, children [WavePlayer_37], properties {attenuationSettings=&/Milkshake_Loop_Drone_Quad_Att}}
 {
  id=Enveloper_2
  children [Mixer_6]
  properties {pitchCurve {editorCurveData {keys [{time=0}]}}, volumeCurve {editorCurveData {keys [<at 0 0 RCIM_Cubic>,<at 3 1>]}}}
 }
 {id=Mixer_6, children [Attenuation_11, Attenuation_12], properties {inputVolume [1.5,0.3]}}
 {id=WavePlayer_36, properties {bLooping=true, soundWaveAssetPtr=&/Milkshake_Ambient_Drone_Loop}}
 {id=WavePlayer_37, properties {bLooping=true, soundWaveAssetPtr=&/Milkshake_Ambient_Drone_Loop_2D_Quad}}
]
properties {
 bOverrideConcurrency=true
 maxDistance=3899.9998
 soundClassObject=/Game/Sounds/FortSoundClassesAndModes/Fort_Earthquakes
 volumeMultiplier=2
 concurrencyOverrides {maxCount=1, resolutionRule=StopFarthestThenPreventNew, voiceStealReleaseTime=3}
}
