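# Sound cue for a Milkshake metal debris rattle effect, routed through the Fort_Earthquakes sound class.
# A quality level node has only its first slot wired, leading to an equally weighted random pick among seven debris wave variations.
# The random node is excluded from branch culling.
# The Milkshake_Rattle_Mono_Att asset handles attenuation, and a concurrency override caps simultaneous instances at three, rejecting new ones beyond that.
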SoundCue /Milkshake/Sounds/Milkshake_Rattle_Metal_Cue
root=QualityLevel_0
nodes [
 {id=QualityLevel_0, children [Random_1,null,null,null]}
 {
  id=Random_1
  children [WavePlayer_7, WavePlayer_8, WavePlayer_9, WavePlayer_10, WavePlayer_11, WavePlayer_12, WavePlayer_13]
  properties {bShouldExcludeFromBranchCulling=true, weights [1,1,1,1,1,1,1]}
 }
 {id=WavePlayer_10, properties {soundWaveAssetPtr=&/milkshake_rattle_metal_debris_04}}
 {id=WavePlayer_11, properties {soundWaveAssetPtr=&/milkshake_rattle_metal_debris_05}}
 {id=WavePlayer_12, properties {soundWaveAssetPtr=&/milkshake_rattle_metal_debris_06}}
 {id=WavePlayer_13, properties {soundWaveAssetPtr=&/milkshake_rattle_metal_debris_07}}
 {id=WavePlayer_7, properties {soundWaveAssetPtr=&/milkshake_rattle_metal_debris_01}}
 {id=WavePlayer_8, properties {soundWaveAssetPtr=&/milkshake_rattle_metal_debris_02}}
 {id=WavePlayer_9, properties {soundWaveAssetPtr=&/milkshake_rattle_metal_debris_03}}
]
properties {
 attenuationSettings=&/Milkshake_Rattle_Mono_Att
 bOverrideConcurrency=true
 maxDistance=2000
 soundClassObject=/Game/Sounds/FortSoundClassesAndModes/Fort_Earthquakes
 volumeMultiplier=1
 concurrencyOverrides {maxCount=3, resolutionRule=PreventNew}
}
