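# Sound cue that plays a stone debris rattle one-shot for the Milkshake content, routed to the Fortnite earthquakes sound class.
# A Quality Level node, with only its first slot connected, feeds a Random node that picks among seven equally weighted wave variations and is excluded from branch culling.
# A shared mono rattle attenuation asset handles spatial falloff, and a concurrency override caps simultaneous instances at a few, blocking new plays beyond that.
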SoundCue /Milkshake/Sounds/Milkshake_Rattle_Stone_Cue
root=QualityLevel_0
nodes [
 {id=QualityLevel_0, children [Random_0,null,null,null]}
 {
  id=Random_0
  children [WavePlayer_0,WavePlayer_1,WavePlayer_2,WavePlayer_3,WavePlayer_4,WavePlayer_5,WavePlayer_6]
  properties {bShouldExcludeFromBranchCulling=true, weights [1,1,1,1,1,1,1]}
 }
 {id=WavePlayer_0, properties {soundWaveAssetPtr=&/milkshake_rattle_stone_debris_01}}
 {id=WavePlayer_1, properties {soundWaveAssetPtr=&/milkshake_rattle_stone_debris_02}}
 {id=WavePlayer_2, properties {soundWaveAssetPtr=&/milkshake_rattle_stone_debris_03}}
 {id=WavePlayer_3, properties {soundWaveAssetPtr=&/milkshake_rattle_stone_debris_04}}
 {id=WavePlayer_4, properties {soundWaveAssetPtr=&/milkshake_rattle_stone_debris_05}}
 {id=WavePlayer_5, properties {soundWaveAssetPtr=&/milkshake_rattle_stone_debris_06}}
 {id=WavePlayer_6, properties {soundWaveAssetPtr=&/milkshake_rattle_stone_debris_07}}
]
properties {
 attenuationSettings=&/Milkshake_Rattle_Mono_Att
 bOverrideConcurrency=true
 maxDistance=2000
 soundClassObject=/Game/Sounds/FortSoundClassesAndModes/Fort_Earthquakes
 volumeMultiplier=1
 concurrencyOverrides {maxCount=3, resolutionRule=PreventNew}
}
