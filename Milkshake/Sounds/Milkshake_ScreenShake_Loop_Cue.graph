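# Looping sound cue for the Milkshake camera shake rumble, played in the Fort_Earthquakes sound class with close, no-pan loop attenuation.
# A looping wave player feeds a continuous modulator, whose volume follows the CameraShakeParam float directly and defaults to silent until that parameter is set.
# An enveloper at the root layers a slow cubic volume fade-in from near silence to full over twelve seconds.
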SoundCue /Milkshake/Sounds/Milkshake_ScreenShake_Loop_Cue
root=Enveloper_2
nodes [
 {
  id=Enveloper_2
  children [ModulatorContinuous_2]
  properties {
   pitchCurve {editorCurveData {keys [{time=0}]}}
   volumeCurve {editorCurveData {keys [<at 0 0.01 RCIM_Cubic>, {arriveTangent=0.1722, interpMode=RCIM_Cubic, leaveTangent=0.1722, tangentMode=RCTM_User, time=12, value=1}]}}
  }
 }
 {
  id=ModulatorContinuous_2
  children [WavePlayer_1]
  properties {
   pitchModulationParams {minOutput=1, parameterName=CameraShakeParam}
   volumeModulationParams {default=0, paramMode=MPM_Direct, parameterName=CameraShakeParam}
  }
 }
 {id=WavePlayer_1, properties {bLooping=true, soundWaveAssetPtr=&/milkshake_camera_loop}}
]
properties {
 attenuationSettings=&/Milkshake_Loop_Close_NoPan_Att
 maxDistance=8000
 soundClassObject=/Game/Sounds/FortSoundClassesAndModes/Fort_Earthquakes
 volumeMultiplier=2.5
}
